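# Unlit masked material for the glowing eyes on the DemonPortal vehicle wheel cosmetic.
# A panning caustics mask, weighted by two oppositely rotated power-curve gradients, wobbles the eye texture UVs.
# The texture's red channel drives opacity mask and emissive, tinted by the Color parameter or constant blue via the Painted static switch.
# The attributes pass through the vehicle damage and de-rez transition functions; several sample and multiply nodes are left unconnected.
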Material /VehicleCosmetics/Wheels/DemonPortal/Materials/M_DemonEyes
graph {
 properties {
  bAutomaticallySetUsageInEditor=false
  bUseMaterialAttributes=true
  bUsedWithSkeletalMesh=true
  blendMode=BLEND_Masked
  shadingModel=MSM_Unlit
  usedShadingModels=Unlit
 }
 nodes [
  n1 = Constant3Vector() {constant=<rgba 0 0 1 0>}
  n2 = VectorParameter() {parameterName=Color, defaultValue=<rgba 1 0.0438 0 1>}
  n3 = ComponentMask(None=MaterialFunctionCall(UVs=TextureCoordinate(), "Rotation Angle (0-1)"=Constant() {r=0.13}) {materialFunction=&CustomRotator}.Rotated Values) {g=true}
  n4 = Constant() {r=12.5}
  n5 = Constant() {r=25}
  n6 = Multiply(A=Power(Base=OneMinus(None=@n3), Exp=@n4), B=@n5)
  n7 = ComponentMask(None=MaterialFunctionCall(UVs=TextureCoordinate(), "Rotation Angle (0-1)"=Constant() {r=-0.13}) {materialFunction=&CustomRotator}.Rotated Values) {g=true}
  n8 = Multiply(A=Power(Base=OneMinus(None=@n7), Exp=@n4), B=@n5)
  n9 = Panner(Coordinate=TextureCoordinate() {uTiling=3, vTiling=3}) {speedY=0.5}
  n10 = ComponentMask(None=TextureSample(UVs=@n9) {texture=/VehicleCosmetics/Wheels/DemonPortal/Textures/CausticsMask_Sharp_RGB_02_T}.RGB) {g=true, r=true}
  n11 = Constant() {r=0.02}
  n12 = Add(A=TextureCoordinate(), B=Multiply(A=Add(A=@n6, B=@n8), B=Multiply(A=@n10, B=@n11)))
  n13 = TextureSample(UVs=@n12) {texture=/VehicleCosmetics/Wheels/DemonPortal/Textures/DemonEyes_RGB_T}
  n14 = Multiply(A=Constant() {r=0.65}, B=Multiply(A=StaticSwitchParameter(True=@n1, False=@n2.RGB) {parameterName=Painted}, B=@n13.R))
  n15 = Multiply(A=@n13.R, B=Constant() {r=1})
  MaterialFunctionCall(In=MaterialFunctionCall(Material=SetMaterialAttributes!(MaterialAttributes, Emissive Color=@n14, Opacity Mask=@n15) {attributeSetTypes [#1, #2]}) {materialFunction=&MF_BaseVehicle_Damage}.Result) {materialFunction=&MF_VehicleDeRez_Transition}
  Multiply(A=VertexColor().A, B=@n13.R)
  Multiply(
   A=@n13.R
   B=MaterialFunctionCall(UVs=TextureCoordinate(), Radius=Constant() {r=0.48}, Density=Constant() {r=10}) {materialFunction=&RadialGradientExponential}.RadialGradientExponential
  )
  TextureSample(UVs=@n12) {samplerType=SAMPLERTYPE_LinearColor, texture=/VehicleCosmetics/Wheels/DemonPortal/Textures/T_DemonPortal_Eye_Alpha}
  TextureSample(UVs=@n9) {texture=/VehicleCosmetics/Wheels/DemonPortal/Textures/PerlinBlur_T}
 ]
}
