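# Unlit translucent smoke material for the DemonPortal vehicle wheel cosmetic, apparently for particles.
# A tiled, scrolling cloud texture, scaled by the DistortStr dynamic parameter, distorts the UVs of a dry-ice smoke texture.
# Vertex color tints the smoke and its alpha scales opacity, with DepthFade softening intersections.
# The resulting material attributes pass through the shared vehicle damage and de-rez transition functions.
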Material /VehicleCosmetics/Wheels/DemonPortal/Materials/M_DemonSmoke
graph {
 properties {
  bAutomaticallySetUsageInEditor=false
  bUseMaterialAttributes=true
  blendMode=BLEND_Translucent
  shadingModel=MSM_Unlit
  usedShadingModels=Unlit
 }
 nodes [
  n1 = VertexColor()
  n2 = TextureCoordinate()
  n3 = Panner(Coordinate=TextureCoordinate() {uTiling=8, vTiling=8}) {speedY=0.06}
  n4 = Multiply(
   A=ComponentMask(None=TextureSample(UVs=@n3) {texture=/VehicleCosmetics/Wheels/DemonPortal/Textures/CloudDistort_T}.RGB) {g=true, r=true}
   B=DynamicParameter() {paramNames [DistortStr,Param2,Param3,Param4]}.DistortStr
  )
  n5 = TextureSample(UVs=Panner(Coordinate=Add(A=@n2, B=@n4))) {texture=/VehicleCosmetics/Wheels/DemonPortal/Textures/dry_ice_01_lp_8x8}
  n6 = Multiply(A=@n1, B=@n5.RGB)
  n7 = DepthFade(Opacity=Multiply(A=@n1.A, B=@n5.A)) {fadeDistanceDefault=3}
  MaterialFunctionCall(In=MaterialFunctionCall(Material=SetMaterialAttributes!(MaterialAttributes, Emissive Color=@n6, Opacity=@n7) {attributeSetTypes [#1, #2]}) {materialFunction=&MF_BaseVehicle_Damage}.Result) {materialFunction=&MF_VehicleDeRez_Transition}
 ]
}
