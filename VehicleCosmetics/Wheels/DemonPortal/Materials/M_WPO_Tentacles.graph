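# Translucent, lit material for the animated tentacles of the DemonPortal vehicle wheel cosmetic, enabled for skeletal meshes, with black base and emissive color.
# A time-driven sine, phased by UV V and vertex color, wiggles World Position Offset, fading as V rises.
# An alpha texture blends opacity between the Lods_Opacity parameter and full.
# The attributes pass through the shared vehicle damage and de-rez transition functions.
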Material /VehicleCosmetics/Wheels/DemonPortal/Materials/M_WPO_Tentacles
graph {
 properties {
  bAutomaticallySetUsageInEditor=false
  bUseMaterialAttributes=true
  bUsedWithSkeletalMesh=true
  blendMode=BLEND_Translucent
  usedShadingModels=Default Lit
 }
 nodes [
  Constant()
  n1 = Constant3Vector()
  n2 = ComponentMask(None=TextureCoordinate()) {g=true}
  n3 = OneMinus(None=@n2)
  n4 = VertexColor()
  n5 = Multiply(A=Time(), B=Constant() {r=0.5})
  n6 = Sine(None=Add(A=Multiply(A=@n2, B=@n4), B=Add(A=@n5, B=@n4))) {period=0.6}
  n7 = TextureSample() {samplerType=SAMPLERTYPE_LinearColor, texture=/VehicleCosmetics/Wheels/DemonPortal/Textures/T_DemonPortal_Tentacles_Alpha_LOD}
  n8 = SetMaterialAttributes!(
   MaterialAttributes
   Base Color=@n1
   Emissive Color=@n1
   World Position Offset=Multiply(A=Clamp(None=@n3), B=ConstantBiasScale(None=@n6) {bias=0.1, scale=0.25})
   Opacity=LinearInterpolate(A=ScalarParameter() {defaultValue=1, parameterName=Lods_Opacity}, B=Constant() {r=1}, Alpha=@n7.R)
  ) {attributeSetTypes [
    #1
    #2
    #3
    #4
   ]}
  MaterialFunctionCall(In=MaterialFunctionCall(Material=@n8) {materialFunction=&MF_BaseVehicle_Damage}.Result) {materialFunction=&MF_VehicleDeRez_Transition}
  Multiply()
  ScalarParameter() {parameterName=Metallic}
  ScalarParameter() {parameterName=Roughness}
  StaticSwitchParameter(True=@n7.R, False=Constant() {r=1}) {parameterName=LODs_Msk}
 ]
}
