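# Unlit translucent material for the demon eyes VFX on the DemonPortal vehicle wheel cosmetic, enabled for Niagara and skeletal meshes.
# Two oppositely rotated, power-sharpened UV gradients mask panning caustics, which distort the eye texture's UVs.
# The eye's red channel is tinted by particle color, or a fixed red through the Painted static switch.
# It also combines with particle alpha and an exponential radial gradient to form a soft-edged mask.
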Material /VehicleCosmetics/Wheels/DemonPortal/VFX/M_DemonEyes
graph {
 properties {
  bAutomaticallySetUsageInEditor=false
  bUsedWithNiagaraMeshParticles=true
  bUsedWithNiagaraRibbons=true
  bUsedWithNiagaraSprites=true
  bUsedWithSkeletalMesh=true
  blendMode=BLEND_Translucent
  previewMesh=/VehicleCosmetics/Wheels/Sunset/Meshes/SK_Wheel_Sunset
  shadingModel=MSM_Unlit
  usedShadingModels=Unlit
 }
 nodes [
  n1 = ComponentMask(None=MaterialFunctionCall(UVs=TextureCoordinate(), "Rotation Angle (0-1)"=Constant() {r=0.13}) {materialFunction=&CustomRotator}.Rotated Values) {g=true}
  n2 = Constant() {r=16}
  n3 = Constant() {r=10}
  n4 = Multiply(A=Power(Base=OneMinus(None=@n1), Exp=@n2), B=@n3)
  n5 = ComponentMask(None=MaterialFunctionCall(UVs=TextureCoordinate(), "Rotation Angle (0-1)"=Constant() {r=-0.13}) {materialFunction=&CustomRotator}.Rotated Values) {g=true}
  n6 = Multiply(A=Power(Base=OneMinus(None=@n5), Exp=@n2), B=@n3)
  n7 = Panner(Coordinate=TextureCoordinate() {uTiling=2, vTiling=2}) {speedY=0.5}
  n8 = ComponentMask(None=TextureSample(UVs=@n7) {texture=/VehicleCosmetics/Wheels/DemonPortal/Textures/CausticsMask_Sharp_RGB_02_T}.RGB) {g=true, r=true}
  n9 = Constant() {r=0.035}
  n10 = Add(A=TextureCoordinate(), B=Multiply(A=Add(A=@n4, B=@n6), B=Multiply(A=@n8, B=@n9)))
  n11 = TextureSample(UVs=@n10) {texture=/VehicleCosmetics/Wheels/DemonPortal/Textures/DemonEyes_RGB_T}
  Multiply(A=VertexColor().A, B=@n11.R)
  n12 = ParticleColor()
  Multiply(
   A=Multiply(A=@n12.A, B=@n11.R)
   B=MaterialFunctionCall(UVs=TextureCoordinate(), Radius=Constant() {r=0.52}, Density=Constant() {r=10}) {materialFunction=&RadialGradientExponential}.RadialGradientExponential
  )
  n13 = Constant3Vector() {constant=<rgba 0.5 0 0 0>}
  Multiply(A=Constant() {r=0.65}, B=Multiply(A=StaticSwitchParameter(True=@n13, False=@n12.RGB) {parameterName=Painted}, B=@n11.R))
  Multiply(A=@n11.R, B=Constant() {r=1})
  TextureSample(UVs=@n10) {samplerType=SAMPLERTYPE_LinearColor, texture=/VehicleCosmetics/Wheels/DemonPortal/Textures/T_DemonPortal_Eye_Alpha}
  TextureSample(UVs=@n7) {texture=/VehicleCosmetics/Wheels/DemonPortal/Textures/PerlinBlur_T}
  VectorParameter() {parameterName=Color, defaultValue=<rgba 1 0.0438 0 1>}
 ]
}
