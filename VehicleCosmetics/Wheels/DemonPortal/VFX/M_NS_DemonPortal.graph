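# Unlit translucent Niagara particle material for the DemonPortal wheel cosmetic's VFX, usable on sprite, ribbon and mesh particles.
# A single Ball_M_T texture sample is multiplied by particle color for the color term and by particle alpha for the opacity term.
# The alpha product passes through a short-distance DepthFade to soften intersections with scene geometry.
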Material /VehicleCosmetics/Wheels/DemonPortal/VFX/M_NS_DemonPortal
graph {
 properties {
  bAutomaticallySetUsageInEditor=false
  bUsedWithNiagaraMeshParticles=true
  bUsedWithNiagaraRibbons=true
  bUsedWithNiagaraSprites=true
  blendMode=BLEND_Translucent
  shadingModel=MSM_Unlit
  usedShadingModels=Unlit
 }
 nodes [
  n1 = ParticleColor()
  n2 = TextureSample() {texture=&/Ball_M_T}
  DepthFade(Opacity=Multiply(A=@n1.A, B=@n2.RGB)) {fadeDistanceDefault=2}
  Multiply(A=@n1.RGB, B=@n2.RGB)
 ]
}
